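# Translucent unlit Niagara material for the blue center effect of the UnluckyRoll pickaxe, usable on sprites, ribbons and mesh particles.
# Noise-warped half-sphere masks, modulated by panning caustics, drive a two-color blend and opacity with angle, edge and depth fades.
# Panning plasma and colored noise add glow before hue and emissive adjustments.
# Dynamic parameters steer UVs and optional spline-thicken, bend and wobble offsets, all lerped to zero by default.
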Material /BRCosmetics/Weapons/FORT_Melee/Pickaxe_UnluckyRoll/FX/M_Pickaxe_UnluckyRoll_BlueCenter
graph {
 properties {
  bUsedWithNiagaraMeshParticles=true
  bUsedWithNiagaraRibbons=true
  bUsedWithNiagaraSprites=true
  blendMode=BLEND_Translucent
  pixelDepthOffsetMode=PDOM_Legacy
  shadingModel=MSM_Unlit
  usedShadingModels=Unlit
 }
 nodes [
  n1 = MaterialFunctionCall(
   WidthBase=ScalarParameter() {defaultValue=0.4, parameterName=SplineBase}
   WidthTip=ScalarParameter() {defaultValue=0.4, parameterName=SplineTip}
   Expand U or V UV Channel=StaticBool()
  ) {materialFunction=&SplineThicken}
  n2 = Constant()
  n3 = ScalarParameter() {defaultValue=1, parameterName=Use_SplineThicken}
  n4 = DynamicParameter() {desc=Parameters set by Niagara, parameterIndex=1, paramNames [X,Y,Z,WPOIntensity]}
  n5 = MaterialFunctionCall(X=@n4.X, Y=@n4.X, Z=@n4.X) {materialFunction=&MakeFloat3}
  n6 = ParticlePositionWS()
  n7 = ScalarParameter() {defaultValue=50, parameterName=Interpolate_WPO_Range}
  n8 = ComponentMask(None=TextureCoordinate()) {r=true}
  n9 = ScalarParameter() {defaultValue=1.5, parameterName=WPOBendPower}
  n10 = LinearInterpolate(
   B=Clamp(None=Subtract(A=@n5.Result, B=@n6), Min=Multiply(A=@n7) {constB=-1}, Max=@n7) {maxDefault=50, minDefault=-50}
   Alpha=Multiply(A=Power(Base=@n8, Exp=@n9), B=@n4.WPOIntensity)
  )
  n11 = Constant()
  n12 = ScalarParameter() {defaultValue=1, parameterName=Use interpolate WPO}
  n13 = DynamicParameter() {parameterIndex=3, defaultValue=<rgba 1 1 1 0>, paramNames [WPOTexture_OffsetU, WPOTexture_OffsetV, WPOTextureU, WPOTextureV]}
  n14 = Add(A=TextureCoordinate() {vTiling=0}, B=AppendVector(A=@n13.WPOTexture_OffsetU, B=@n13.WPOTexture_OffsetU))
  n15 = AppendVector(A=@n13.WPOTextureU, B=@n13.WPOTextureU)
  n16 = Constant2Vector() {r=-2}
  n17 = ScalarParameter() {defaultValue=1, group=WPO, parameterName=WPOPanner_Speed}
  n18 = TextureSampleParameter2D(UVs=Panner(Coordinate=Multiply(A=@n14, B=@n15), Speed=Multiply(A=@n16, B=@n17)) {speedX=-2}) {
   group=WPO
   parameterName=WPOTexture
   samplerType=SAMPLERTYPE_Normal
   texture=&T_Weap_Distort_N_02
  }
  n19 = TextureCoordinate()
  n20 = Multiply(
   A=@n18.R
   B=LinearInterpolate(A=ScalarParameter() {defaultValue=0.05, group=WPO, parameterName=WPOMaskLerp}, Alpha=ComponentMask(None=@n19) {r=true}) {constA=0.05}
  )
  n21 = Multiply(A=Transform(None=Constant3Vector() {constant=<rgba 1 0 0 0>}) {transformSourceType=TRANSFORMSOURCE_Local}, B=@n20)
  n22 = ScalarParameter() {defaultValue=1, group=WPO, parameterName=X WPO_Intensity}
  n23 = ScalarParameter() {defaultValue=1, group=WPO, parameterName=WPO XYZ IntensityControl}
  n24 = Multiply(A=Transform(None=Constant3Vector() {constant=<rgba 0 1 0 0>}) {transformSourceType=TRANSFORMSOURCE_Local}, B=@n20)
  n25 = ScalarParameter() {defaultValue=1, group=WPO, parameterName=Y WPO_Intensity}
  n26 = Add(A=Multiply(A=Multiply(A=@n21, B=@n22) {constB=10}, B=@n23), B=Multiply(A=Multiply(A=@n24, B=@n25) {constB=10}, B=@n23))
  n27 = Transform(None=Constant3Vector() {constant=<rgba 0 0 1 0>}) {transformSourceType=TRANSFORMSOURCE_Local}
  n28 = Multiply(A=Multiply(A=Multiply(A=@n27, B=@n20), B=ScalarParameter() {defaultValue=1, group=WPO, parameterName=Z WPO_Intensity}) {constB=10}, B=@n23)
  Add(
   A=Add(A=LinearInterpolate(A=@n1.WorldPositionOffset, B=@n2, Alpha=@n3), B=LinearInterpolate(A=@n10, B=@n11, Alpha=@n12))
   B=LinearInterpolate(A=Add(A=@n26, B=@n28), B=Constant(), Alpha=ScalarParameter() {defaultValue=1, group=WPO, parameterName=Use WPO})
  )
  Constant()
  n29 = ParticleColor()
  n30 = TextureCoordinate() {uTiling=0.8, vTiling=0.5}
  n31 = TextureCoordinate()
  n32 = TextureCoordinate() {uTiling=0.5, vTiling=1.3}
  n33 = DynamicParameter() {paramNames [LerpTextureUV_Offset, LerpTextureU_Offset_Add, Param3, Param4]}
  n34 = LinearInterpolate(
   A=@n32
   B=Add(
    A=@n32
    B=MaterialFunctionCall(X=@n33.LerpTextureU_Offset_Add, Y=@n33.LerpTextureU_Offset_Add) {materialFunction=&MakeFloat2}.Result
   )
   Alpha=ScalarParameter() {parameterName=UseDynamicUOffset}
  )
  n35 = Multiply(A=Constant2Vector() {r=-1}, B=ScalarParameter() {defaultValue=1, parameterName=MaskTexture_Speed_Master})
  n36 = LinearInterpolate(Alpha=TextureSample(UVs=Panner(Coordinate=@n34, Speed=@n35) {speedY=1}) {samplerType=SAMPLERTYPE_LinearColor, texture=&T_NoisePack_LinearColor}.R) {constA=-1, constB=0.5}
  n37 = AppendVector(A=Add(A=ComponentMask(None=@n30) {r=true}) {constB=0}, B=Add(A=ComponentMask(None=@n31) {g=true}, B=Multiply(A=@n36) {constB=0.2}))
  n38 = Constant2Vector() {g=0.1, r=0.35}
  n39 = Power(Base=TextureSample(UVs=Add(A=@n37, B=@n38) {constB=0}) {samplerSource=SSM_Clamp_WorldGroupSettings, texture=/Game/Effects/Fort_Effects/Textures/Mask/T_HalfSphere_D}.R) {constExponent=5}
  n40 = ComponentMask(None=TextureCoordinate() {uTiling=0.7, vTiling=0.5}) {r=true}
  n41 = TextureCoordinate() {uTiling=0.8, vTiling=-1.3}
  n42 = DynamicParameter() {paramNames [LerpTextureUV_Offset, LerpTextureU_Offset_Add, Param3, Param4]}
  n43 = Constant()
  n44 = LinearInterpolate(
   A=@n41
   B=Add(
    A=@n41
    B=MaterialFunctionCall(X=@n42.LerpTextureU_Offset_Add, Y=@n43) {materialFunction=&MakeFloat2}.Result
   )
   Alpha=ScalarParameter() {parameterName=UseMaskOffset_Add}
  )
  n45 = Multiply(A=TextureSample(UVs=Panner(Coordinate=@n44, Speed=@n35) {speedY=1}) {samplerType=SAMPLERTYPE_LinearColor, texture=&T_NoisePack_LinearColor}.R) {constB=-1}
  n46 = Add(A=ComponentMask(None=TextureCoordinate()) {g=true}, B=Multiply(A=LinearInterpolate(Alpha=@n45) {constA=-1, constB=0.5}) {constB=0.2})
  n47 = TextureSample(UVs=Add(A=AppendVector(A=@n40, B=@n46), B=Constant2Vector() {g=0.35, r=0.4}) {constB=0}) {samplerSource=SSM_Clamp_WorldGroupSettings, texture=&T_HalfSphere_D}
  n48 = LinearInterpolate(Alpha=Saturate(None=Add(A=@n39, B=@n47.R) {constB=0}))
  n49 = ScalarParameter() {defaultValue=0.1, parameterName=LerpTexture_fusion}
  n50 = ScalarParameter() {defaultValue=1, parameterName=LerpTexture_Lightfusion}
  n51 = DynamicParameter() {paramNames [LerpTextureUV_Offset, LerpTextureU_Offset_Add, Param3, Param4]}
  n52 = Multiply(A=ScalarParameter() {defaultValue=1, parameterName=LerpTextureU}, B=@n51.LerpTextureUV_Offset)
  n53 = VertexColor()
  n54 = ScalarParameter() {defaultValue=1, parameterName=LerpTextureV}
  n55 = Multiply(A=TextureCoordinate(), B=AppendVector(A=Multiply(A=@n52, B=@n53.R), B=Multiply(A=@n54, B=@n51.LerpTextureUV_Offset)))
  n56 = Add(
   A=@n55
   B=MaterialFunctionCall(X=@n51.LerpTextureU_Offset_Add, Y=Constant()) {materialFunction=&MakeFloat2}.Result
  )
  n57 = ScalarParameter() {parameterName=UseMaskOffset_Add}
  n58 = Constant2Vector() {r=-1}
  n59 = Multiply(A=ScalarParameter() {defaultValue=1, parameterName=Texture_Speed_Master}, B=VertexColor().R)
  n60 = TextureSampleParameter2D(UVs=Panner(Coordinate=LinearInterpolate(A=@n55, B=@n56, Alpha=@n57), Speed=Multiply(A=@n58, B=@n59)) {speedX=-1}) {
   parameterName=LerpTexture
   samplerType=SAMPLERTYPE_LinearGrayscale
   texture=/Game/Global/Textures/Default/Caustics/T_Nosie_Caustics_Grayscale
  }
  n61 = Multiply(A=@n48, B=Saturate(None=LinearInterpolate(A=@n49, B=@n50, Alpha=@n60.R) {constA=0.1}))
  n62 = Power(Base=Multiply(A=Multiply(A=@n29.A, B=@n61)))
  n63 = ScalarParameter() {parameterName=AngleFade}
  n64 = TextureCoordinate()
  n65 = DynamicParameter() {parameterIndex=2, defaultValue=<rgba 0 1 1 1>, paramNames [Left linear soft U add, Param2, Param3, Param4]}
  n66 = AppendVector(A=Add(A=ScalarParameter() {parameterName=Left linear soft U add}, B=@n65.Left linear soft U add), B=Constant())
  n67 = Multiply(A=ComponentMask(None=Add(A=@n64, B=@n66)) {r=true}) {constB=15}
  n68 = Multiply(
   A=Multiply(
    A=@n62
    B=MaterialFunctionCall("FadeSensitivity (optional)"=@n63) {materialFunction=&FadeBasedOnViewAngle}.Result
   )
   B=Saturate(None=Power(Base=@n67) {constExponent=1.5})
  )
  n69 = DepthFade(Opacity=@n68, FadeDistance=ScalarParameter() {defaultValue=5, parameterName=DepthFade})
  NamedRerouteDeclaration(None=ShadingPathSwitch(Default=@n69, Deferred=@n69, Forward=@n68, Mobile=@n68)) {name=Opacity, nodeColor=<rgba 1 0.8203 0 1>}
  n70 = ScalarParameter() {parameterName=OverallHueShift}
  n71 = Multiply(
   A=LinearInterpolate(
    A=VectorParameter() {parameterName=color01, defaultValue=<rgba 0.0035 0 0.0729 1>}.RGB
    B=VectorParameter() {parameterName=color02, defaultValue=<rgba 0 0.0172 1 1>}.RGB
    Alpha=@n48
   )
   B=@n29.RGB
  )
  n72 = TextureCoordinate() {uTiling=1.5}
  n73 = DynamicParameter() {paramNames [LerpTextureUV_Offset, LerpTextureU_Offset_Add, Param3, Param4]}
  n74 = Constant()
  n75 = LinearInterpolate(
   A=@n72
   B=Add(
    A=@n72
    B=MaterialFunctionCall(X=@n73.LerpTextureU_Offset_Add, Y=@n74) {materialFunction=&MakeFloat2}.Result
   )
   Alpha=ScalarParameter() {parameterName=UseMaskOffset_Add}
  )
  n76 = Multiply(A=Constant2Vector() {g=0.5, r=-1}, B=ScalarParameter() {defaultValue=1, parameterName=Texture_Speed_Master})
  n77 = LinearInterpolate(Alpha=TextureSample(UVs=Panner(Coordinate=@n75, Speed=@n76) {speedX=-1, speedY=0.5}) {samplerType=SAMPLERTYPE_LinearGrayscale, texture=/Game/Effects/Fort_Effects/Emotes/UltraLight/Texture/T_Plasma_C_Greyscale}.R)
  n78 = TextureCoordinate()
  n79 = DynamicParameter() {paramNames [LerpTextureUV_Offset, LerpTextureU_Offset_Add, Param3, Param4]}
  n80 = Constant()
  n81 = LinearInterpolate(
   A=@n78
   B=Add(
    A=@n78
    B=MaterialFunctionCall(X=@n79.LerpTextureU_Offset_Add, Y=@n80) {materialFunction=&MakeFloat2}.Result
   )
   Alpha=ScalarParameter() {parameterName=UseMaskOffset_Add}
  )
  n82 = Multiply(A=Constant2Vector() {r=-3}, B=ScalarParameter() {defaultValue=1, parameterName=Texture_Speed_Master})
  n83 = Multiply(A=TextureSample(UVs=Panner(Coordinate=@n81, Speed=@n82) {speedX=-3}) {texture=/Game/Effects/Fort_Effects/Emotes/Energize_Max/LV4/T_Energize_Color_Noise_blur}.RGB)
  n84 = ScalarParameter() {defaultValue=-5, parameterName=Des}
  n85 = Multiply(A=Multiply(A=Saturate(None=@n77), B=Desaturation(None=@n83, Fraction=@n84)), B=@n62)
  n86 = Desaturation(
   None=LinearInterpolate(A=@n71, B=Add(A=@n85, B=@n71), Alpha=ScalarParameter() {defaultValue=1, parameterName=Use_LightEnergy})
   Fraction=ScalarParameter() {parameterName=OverallDesatuation}
  )
  NamedRerouteDeclaration(None=MaterialFunctionCall(Emissive=MaterialFunctionCall(Hue Shift Percentage=@n70, Texture=@n86) {materialFunction=&HueShift}.Result) {materialFunction=&MF_FX_EmissiveAdjustments}.Result) {name=Emissive, nodeColor=<rgba 0 1 0.5781 1>}
  NamedRerouteUsage()
  NamedRerouteUsage()
 ]
}
