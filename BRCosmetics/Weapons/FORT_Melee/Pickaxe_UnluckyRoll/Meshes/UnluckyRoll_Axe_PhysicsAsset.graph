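# Physics asset for the UnluckyRoll pickaxe axe mesh, a two-handed melee cosmetic.
# It has a single body on the Melee_TwoHanded_Handle bone, shaped as one capsule that is offset and rotated from that bone.
# It defines no constraints, so it has no jointed or separately simulated parts.
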PhysicsAsset /BRCosmetics/Weapons/FORT_Melee/Pickaxe_UnluckyRoll/Meshes/UnluckyRoll_Axe_PhysicsAsset
bodies {Melee_TwoHanded_Handle [{shapeType=Capsule, length=60.5896, radius=8.4644, center=<0.0001 2.1539 31.9397>, rotation=<rot 2.9413 -90 0.0002>}]}
constraints []
